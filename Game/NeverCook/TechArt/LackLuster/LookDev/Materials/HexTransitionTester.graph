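# Look-dev test material for an animated hexagon transition, with no connections to material outputs shown.
# Tiled hex mask values offset two modulo-time clocks into sines, making per-channel waves that scale the hex normal map's X and Y before renormalizing.
# A separate branch multiplies two opposing panned noise samples and compares them to the hex mask through a SphereMask, selecting areas where values nearly match.
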Material /Game/NeverCook/TechArt/LackLuster/LookDev/Materials/HexTransitionTester
graph {
 properties {pixelDepthOffsetMode=PDOM_Legacy, usedShadingModels=Default Lit}
 nodes [
  Constant() {r=0.5}
  Constant() {r=0.8}
  Constant() {r=0.2}
  n1 = TextureCoordinate() {uTiling=5, vTiling=5}
  n2 = TextureSample(UVs=@n1) {samplerType=SAMPLERTYPE_LinearColor, texture=/Game/NeverCook/TechArt/LackLuster/LookDev/Textures/Edge_Hexagons}
  n3 = MaterialFunctionCall!(Time, Divisor=Constant() {r=3}, Mod) {materialFunction=/Game/Effects/Fort_Effects/Materials/_Functions/MF_ModuloTime}
  n4 = MaterialFunctionCall(Divisor=Constant() {r=4.5}) {materialFunction=&MF_ModuloTime}
  n5 = MaterialFunctionCall(R=Sine(None=Add(A=@n2.R, B=@n3.Unclamped)) {period=0.5}, G=Sine(None=Add(A=@n2.R, B=@n4.Unclamped)) {period=0.5}, B=Constant() {r=1}) {materialFunction=&AppendMany}
  Normalize(VectorInput=Multiply(
    A=TextureSample(UVs=@n1) {samplerType=SAMPLERTYPE_Normal, texture=/Game/NeverCook/TechArt/LackLuster/LookDev/Textures/Edge_NormalHexagons}.RGB
    B=ConstantBiasScale(None=@n5.RGB)
   ))
  n6 = TextureCoordinate()
  n7 = Panner(Coordinate=@n6) {bFractionalPart=true, speedX=0.005, speedY=0.0025}
  n8 = Panner(Coordinate=@n6) {bFractionalPart=true, speedX=-0.0187, speedY=-0.0015}
  SphereMask(
   A=Multiply(
    A=TextureSample(UVs=@n7) {samplerType=SAMPLERTYPE_LinearColor, texture=/Game/Global/Textures/Default/Noise/T_Noise01}.R
    B=TextureSample(UVs=@n8) {samplerType=SAMPLERTYPE_LinearColor, texture=&T_Noise01}.B
   )
   B=TextureSample() {samplerType=SAMPLERTYPE_Grayscale, texture=&Edge_Hexagons}.R
   Radius=Constant() {r=0.1}
   Hardness=Constant()
  )
 ]
}
