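# Material function that layers a chrome look over incoming material attributes, named for landscape RVT use.
# It overrides base color, roughness and specular, leaving metallic untouched, then blends the result with the original by a Mask input.
# A grayscale texture drives roughness between low and high parameters.
# World-position XY over a divisor supplies tiling, scalable UVs, a top-down planar projection despite the "CheapTriplanar" parameter name.
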MaterialFunction /Game/NeverCook/TechArt/LackLuster/LookDev/Materials/MF_Chrome_Landscape_RVT
graph {nodes [
  n1 = FunctionInput() {inputName=MA, inputType=FunctionInput_MaterialAttributes}
  n2 = VectorParameter() {parameterName=ChromeBaseColor_RVT, defaultValue=<rgba 0.0776 0.1134 0.1476 0>}
  n3 = LinearInterpolate(
   A=ScalarParameter() {parameterName=RoughnessLow}
   B=ScalarParameter() {defaultValue=0.7, parameterName=RoughnessHigh}
   Alpha=TextureSampleParameter2D(UVs=NamedRerouteUsage()) {parameterName=Roughness, samplerType=SAMPLERTYPE_LinearGrayscale, texture=null}.R
  )
  n4 = ScalarParameter() {parameterName=Specular}
  FunctionOutput(None=BlendMaterialAttributes(
    A=@n1
    B=SetMaterialAttributes!(MaterialAttributes=@n1, BaseColor=@n2.RGB, Roughness=@n3, Specular=@n4) {attributeSetTypes [#1, #2, #3]}
    Alpha=FunctionInput() {bUsePreviewValueAsDefault=true, inputName=Mask, inputType=FunctionInput_Scalar, sortPriority=1}
   ))
  n5 = ScalarParameter() {defaultValue=4, parameterName=Base_Scale_U}
  n6 = ScalarParameter() {defaultValue=4, parameterName=Base_Scale_V}
  NamedRerouteDeclaration(None=Multiply(A=NamedRerouteUsage(), B=AppendVector(A=@n5, B=@n6))) {name=BaseUVs, nodeColor=<rgba 0 1 0.625 1>}
  n7 = WorldPosition()
  n8 = ScalarParameter() {defaultValue=1000, parameterName=CheapTriplanarDivisor}
  NamedRerouteDeclaration(None=ComponentMask(None=Divide(A=@n7.XYZ, B=@n8)) {g=true, r=true}) {name=UVs, nodeColor=<rgba 0.2891 0 1 1>}
 ]}
