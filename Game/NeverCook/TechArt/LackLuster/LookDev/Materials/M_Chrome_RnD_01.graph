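# Lookdev R&D material for an animated chrome surface mixing iron PBR maps with tinted marble, caustics and sparkles.
# Reflection-vector-offset swirl and time-cycled caustic textures index curve atlas rows for color, with Fresnel driving hue shift and edge shaping.
# Panning noise on floor-quantized UVs forms blocky transition masks, alongside blended normal, roughness and metallic layers.
# Includes optional world-aligned UVs and two experimental edge-detection chains using offset sampling and DDX/DDY.
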Material /Game/NeverCook/TechArt/LackLuster/LookDev/Materials/M_Chrome_RnD_01
graph {
 properties {pixelDepthOffsetMode=PDOM_Legacy, usedShadingModels=Default Lit}
 nodes [
  n1 = TextureSampleParameter2D(UVs=NamedRerouteUsage()) {
   parameterName=Albedo
   samplerType=SAMPLERTYPE_LinearColor
   texture=/Game/NeverCook/TechArt/LackLuster/LookDev/Textures/Iron/se4objgc_4K_Albedo
  }
  n2 = Desaturation(None=@n1.RGB, Fraction=ScalarParameter() {parameterName=AlbedoDesaturationAmount})
  n3 = VectorParameter() {parameterName=AlbedoBlocksTransition, defaultValue=<rgba 0.9045 0.8794 0.7957 0>}
  n4 = MaterialFunctionCall(Exponent=ScalarParameter() {defaultValue=3, parameterName=FresnelHueShiftExponent}, Vertex / Pixel Normal=PixelNormalWS()) {materialFunction=&MF_Fresnel_FX}
  n5 = Multiply(
   A=LinearInterpolate(A=@n2, B=Multiply(A=@n2, B=@n3.RGB), Alpha=NamedRerouteUsage())
   B=MaterialFunctionCall(Hue Shift Percentage=@n4.Fresnel, Texture=VectorParameter() {parameterName=AlbedoScale, defaultValue=<rgba 0.9596 1 0.8663 0>}.RGB) {materialFunction=&HueShift}.Result
  )
  n6 = NamedRerouteUsage()
  n7 = Multiply(
   A=ComponentMask(None=ReflectionVectorWS(CustomWorldNormal=@n6)) {g=true, r=true}
   B=ScalarParameter() {defaultValue=0.04, parameterName=RelectionVectorStrength}
  ) {constB=0.1}
  n8 = NamedRerouteUsage()
  n9 = CurveAtlasRowParameter(CurveTime=TextureSample(UVs=Add(A=@n7, B=@n8)) {samplerType=SAMPLERTYPE_LinearColor, texture=/Game/NeverCook/TechArt/LackLuster/LookDev/Textures/ChromeSwirl_output}.R) {
   atlas=/Game/NeverCook/TechArt/LackLuster/LookDev/Textures/CA_Chrome_RnD
   curve=/Game/NeverCook/TechArt/LackLuster/LookDev/Textures/C_Chrome_02
   defaultValue=1
   parameterName=Chrome1
  }
  n10 = Multiply(A=Multiply(A=@n9, B=ScalarParameter() {defaultValue=1, parameterName=ChromeEdgeScale}), B=NamedRerouteUsage())
  n11 = Multiply(A=@n10, B=ScalarParameter() {defaultValue=0.005, parameterName=ChromeScaleEmissive})
  n12 = ScalarParameter() {defaultValue=0.25, parameterName=CausticsEmissiveScale}
  n13 = NamedRerouteUsage()
  n14 = Multiply(A=TextureSample(UVs=@n13) {samplerType=SAMPLERTYPE_LinearColor, texture=/Game/NeverCook/TechArt/LackLuster/LookDev/Textures/ChromeTech_Caustics}.G) {constB=0.5}
  n15 = ScalarParameter() {defaultValue=4.25, parameterName=CausticsTimeDivisor}
  n16 = Add(
   A=@n4.Fresnel
   B=Add(
    A=TextureSample(UVs=@n13) {samplerType=SAMPLERTYPE_LinearColor, texture=&ChromeTech_Caustics}.R
    B=MaterialFunctionCall!(Time, Divisor=@n15, Mod) {materialFunction=/Game/Effects/Fort_Effects/Materials/_Functions/MF_ModuloTime}.Clamped 0-1
   )
  )
  n17 = CurveAtlasRowParameter(CurveTime=Frac(None=Add(A=@n14, B=@n16))) {
   atlas=&CA_Chrome_RnD
   curve=/Game/NeverCook/TechArt/LackLuster/LookDev/Textures/C_Chrome_04
   defaultValue=3
   parameterName=Caustics
  }
  n18 = NamedRerouteUsage()
  n19 = Multiply(A=@n17, B=@n18)
  n20 = Multiply(A=ScalarParameter() {defaultValue=60, parameterName=SparklesScale}, B=NamedRerouteUsage())
  Add(
   A=Add(A=Multiply(A=@n5, B=@n11), B=Multiply(A=@n12, B=@n19)) {constB=0}
   B=Multiply(A=Multiply(A=@n20, B=@n17), B=Power(Base=@n18) {constExponent=4})
  )
  n21 = TextureSampleParameter2D() {parameterName=Diffuse, texture=/Game/Weapons/WeaponSkins/Wraps/Textures/BlackOut1/TBlackOut2_D}
  n22 = ScalarParameter() {defaultValue=0.7, parameterName=EdgeColorScale}
  n23 = Constant() {r=1}
  n24 = Add(A=TextureCoordinate(), B=ScalarParameter() {defaultValue=0.003, parameterName=EdgeDetectionOffset})
  n25 = Subtract(
   A=Desaturation(None=@n21.RGB, Fraction=@n23)
   B=Desaturation(
    None=TextureSampleParameter2D(UVs=@n24) {parameterName=Diffuse, texture=&TBlackOut2_D}.RGB
    Fraction=@n23
   )
  )
  n26 = Saturate(None=Multiply(A=Saturate(None=@n25), B=ScalarParameter() {defaultValue=20, parameterName=EdgeDetectionScale}))
  n27 = ScalarParameter() {defaultValue=2, parameterName=EdgeDetectionExponent}
  LinearInterpolate(
   A=@n21.RGB
   B=LinearInterpolate(B=Multiply(B=@n22), Alpha=Power(Base=@n26, Exp=@n27) {constExponent=3})
   Alpha=ScalarParameter() {defaultValue=1, parameterName=testBlend}
  )
  n28 = Constant() {r=3}
  n29 = Constant3Vector() {constant=<rgba 0 0 1 0>}
  n30 = TextureSample(UVs=NamedRerouteUsage()) {samplerType=SAMPLERTYPE_LinearColor, texture=/Game/NeverCook/TechArt/LackLuster/LookDev/Textures/ZiaratWhiteMarble/tgzkdehv_4K_Albedo}
  n31 = ScalarParameter() {defaultValue=1.4, parameterName=MarbleFresnelBrightness}
  n32 = ScalarParameter() {defaultValue=2, parameterName=MarbleFresnelExponent}
  n33 = Multiply(
   A=@n30.RGB
   B=Multiply(
    A=VectorParameter() {parameterName=MarbleColor, defaultValue=<rgba 0.4656 0.6775 0.9948 0>}.RGB
    B=MaterialFunctionCall(Brightness=@n31, Exponent=@n32) {materialFunction=&MF_Fresnel_FX}.Fresnel
   )
  )
  Multiply(
   A=LinearInterpolate(
    A=ScalarParameter() {defaultValue=0.6, parameterName=-Zscale}
    Alpha=MaterialFunctionCall(Brightness=@n28, Camera Vector=@n29) {materialFunction=&MF_Fresnel_FX}.Fresnel
   )
   B=LinearInterpolate(A=Add(A=@n10, B=@n5), B=Add(A=@n33, B=@n19), Alpha=NamedRerouteUsage())
  )
  n34 = ScalarParameter() {defaultValue=4, parameterName=Base_Scale_U}
  n35 = ScalarParameter() {defaultValue=4, parameterName=Base_Scale_V}
  NamedRerouteDeclaration(None=Multiply(A=NamedRerouteUsage(), B=AppendVector(A=@n34, B=@n35))) {name=Base_UVs, nodeColor=<rgba 1 0.0234 0 1>}
  n36 = TextureSampleParameter2D() {parameterName=Normals, samplerType=SAMPLERTYPE_Normal, texture=&FlatNormalBlue}
  n37 = Multiply(A=@n36.RGB, B=VectorParameter() {parameterName=NormalsStrength, defaultValue=<rgba 0.15 0.15 1 0>}.RGB)
  n38 = TextureSampleParameter2D(UVs=NamedRerouteUsage()) {
   parameterName=NormalDetail
   samplerType=SAMPLERTYPE_Normal
   texture=/Game/NeverCook/TechArt/LackLuster/LookDev/Textures/RustySheetMetal/vbuifjj_4K_Normal
  }
  n39 = Multiply(A=@n38.RGB, B=Constant3Vector() {constant=<rgba 1 1 0 0>})
  n40 = TextureSample(UVs=NamedRerouteUsage()) {samplerType=SAMPLERTYPE_Normal, texture=/Game/NeverCook/TechArt/LackLuster/LookDev/Textures/ChromeSwirl_Normal}
  n41 = Multiply(
   A=TextureSample(UVs=NamedRerouteUsage()) {samplerType=SAMPLERTYPE_Normal, texture=&ChromeSwirl_Normal}.RGB
   B=Constant3Vector() {constant=<rgba 1 1 0 0>}
  )
  n42 = Multiply(A=VectorParameter() {parameterName=EdgeNormalScale, defaultValue=<rgba 0.8 0.8 1 0>}.RGB, B=Normalize(VectorInput=Add(A=@n40.RGB, B=@n41)))
  n43 = VertexNormalWS()
  n44 = NamedRerouteUsage()
  n45 = ScalarParameter() {defaultValue=0.15, parameterName=EdgeNormal_Blend}
  n46 = MaterialFunctionCall(
   Exponent=ScalarParameter() {defaultValue=3.5, parameterName=EdgeNormal_FresnelExponent}
   Vertex / Pixel Normal=Normalize(VectorInput=LinearInterpolate(A=@n43, B=@n44, Alpha=@n45))
  ) {materialFunction=&MF_Fresnel_FX}
  n47 = LinearInterpolate(Alpha=SphereMask(
    A=NamedRerouteDeclaration(None=@n46.Fresnel Inverted) {name=EdgeNormalFresnel, nodeColor=<rgba 0 1 0.5547 1>}
    B=Constant() {r=0.5}
    Radius=ScalarParameter() {defaultValue=0.35, parameterName=EdgeNormalFresnelRadius}
    Hardness=Constant()
   )) {constA=0.05, desc=Meh?}
  n48 = LinearInterpolate(
   A=Add(A=Add(A=@n37, B=@n39), B=LinearInterpolate(B=@n42, Alpha=@n47))
   B=Multiply(A=@n36.RGB, B=VectorParameter() {parameterName=NormalsCausticStrength, defaultValue=<rgba 0.15 0.15 1 0>}.RGB)
   Alpha=NamedRerouteUsage()
  )
  n49 = AppendVector(A=ScalarParameter() {defaultValue=32, parameterName=BlocksScaleU}, B=ScalarParameter() {defaultValue=16, parameterName=BlocksScaleV})
  n50 = TextureSample(UVs=Multiply(A=NamedRerouteUsage(), B=NamedRerouteDeclaration(None=@n49) {name=Blocks.UV, nodeColor=<rgba 1 0.5625 0 1>})) {samplerType=SAMPLERTYPE_Normal, texture=/Game/NeverCook/TechArt/LackLuster/LookDev/Textures/Edge_Normal}
  n51 = NamedRerouteUsage()
  NamedRerouteDeclaration(None=Normalize(VectorInput=LinearInterpolate(A=@n48, B=@n50.RGB, Alpha=@n51))) {name=Normals, nodeColor=<rgba 1 0 0.9141 1>}
  n52 = ScalarParameter() {defaultValue=1, parameterName=Caustics_Scale_U}
  n53 = ScalarParameter() {defaultValue=0.5, parameterName=Caustics_Scale_V}
  NamedRerouteDeclaration(None=Multiply(A=NamedRerouteUsage(), B=AppendVector(A=@n52, B=@n53))) {name=CausticsUVs, nodeColor=<rgba 0.9453 1 0 1>}
  n54 = ScalarParameter() {defaultValue=1, parameterName=CausticsTransitionU_Scale}
  n55 = ScalarParameter() {defaultValue=1, parameterName=CausticsTransitionV_Scale}
  n56 = Multiply(A=NamedRerouteUsage()) {constB=8}
  n57 = Multiply(A=Multiply(A=NamedRerouteUsage(), B=AppendVector(A=@n54, B=@n55)), B=@n56) {constB=8}
  n58 = NamedRerouteDeclaration(None=Divide(A=Floor(None=@n57), B=@n56)) {name=Caustics.TransitionBlocks.UVs, nodeColor=<rgba 1 0.9141 0 1>}
  n59 = TextureSample(UVs=Panner(Coordinate=@n58) {bFractionalPart=true, speedX=0.005, speedY=0.0025}) {samplerType=SAMPLERTYPE_LinearColor, texture=/Game/Global/Textures/Default/Noise/T_Noise01}
  n60 = TextureSample(UVs=Panner(Coordinate=@n58) {bFractionalPart=true, speedX=-0.0187, speedY=-0.0015}) {samplerType=SAMPLERTYPE_LinearColor, texture=&T_Noise01}
  n61 = SphereMask(
   A=Saturate(None=Multiply(A=@n59.R, B=@n60.B))
   B=ScalarParameter() {defaultValue=0.65, parameterName=BlocksNoiseCenter}
   Radius=ScalarParameter() {defaultValue=0.5, parameterName=BlocksNoiseRadius}
   Hardness=ScalarParameter() {defaultValue=0.6, parameterName=BlocksNoiseHardness}
  )
  NamedRerouteDeclaration(None=SphereMask(
    A=@n61
    B=ScalarParameter() {defaultValue=0.5, parameterName=BlocksTransitionCenter}
    Radius=ScalarParameter() {defaultValue=0.5, parameterName=BlocksTransitionRadius}
    Hardness=ScalarParameter() {defaultValue=0.1, parameterName=BlocksTransitionHardness}
   )) {name=BlocksTransition, nodeColor=<rgba 0.0781 1 0 1>}
  n62 = Multiply(
   A=NamedRerouteUsage()
   B=AppendVector(A=ScalarParameter() {defaultValue=20, parameterName=SparklesTileU}, B=ScalarParameter() {defaultValue=10, parameterName=SparklesTileV})
  )
  n63 = TextureSample(UVs=@n62) {samplerType=SAMPLERTYPE_LinearColor, texture=/Game/NeverCook/TechArt/LackLuster/LookDev/Textures/Sparkle_Normal}
  n64 = Abs(None=DotProduct(A=Transform(None=@n63.RGB), B=CameraVectorWS()))
  n65 = ScalarParameter() {defaultValue=60, parameterName=SparklesExponent}
  NamedRerouteDeclaration(None=Multiply(
    A=Power(Base=@n64, Exp=@n65) {constExponent=20}
    B=TextureSample(UVs=@n62) {samplerType=SAMPLERTYPE_LinearColor, texture=/Game/NeverCook/TechArt/LackLuster/LookDev/Textures/Sparkle_Mask}.R
   )) {name=Sparkles, nodeColor=<rgba 1 0.5391 0 1>}
  n66 = Divide(A=WorldPosition().XYZ, B=ScalarParameter() {defaultValue=1000, parameterName=CheapTriplanarDivisor})
  NamedRerouteDeclaration(None=StaticSwitchParameter(
    True=MaterialFunctionCall!(UseLocalSpace, WorldPosition=@n66, NormalOverride) {materialFunction=/Game/Environments/Apollo/Sets/Skyfire/Materials/MaterialFunctions/MF_WorldAlignedUVs_HardEdge}.Result
    False=TextureCoordinate()
   ) {parameterName=UseCheapTriplanarUVs}) {name=UVs, nodeColor=<rgba 0.2891 0 1 1>}
  n67 = TextureSampleParameter2D(UVs=NamedRerouteUsage()) {
   parameterName=Roughness
   samplerType=SAMPLERTYPE_LinearColor
   texture=/Game/NeverCook/TechArt/LackLuster/LookDev/Textures/Iron/se4objgc_4K_Roughness
  }
  n68 = LinearInterpolate(A=ScalarParameter() {parameterName=RoughnessLow}, B=ScalarParameter() {defaultValue=1, parameterName=RoughnessHigh}, Alpha=@n67.R)
  n69 = ScalarParameter() {defaultValue=0.33, parameterName=RoughnessCaustics}
  n70 = NamedRerouteUsage()
  NamedRerouteDeclaration(None=LinearInterpolate(
    A=LinearInterpolate(A=@n68, B=@n69, Alpha=@n70)
    B=ScalarParameter() {defaultValue=0.25, parameterName=RoughnessTransition}
    Alpha=NamedRerouteUsage()
   )) {name=Roughness, nodeColor=<rgba 1 0.3984 0 1>}
  n71 = TextureSampleParameter2D(UVs=NamedRerouteUsage()) {
   parameterName=Metallic
   samplerType=SAMPLERTYPE_LinearColor
   texture=/Game/NeverCook/TechArt/LackLuster/LookDev/Textures/Iron/se4objgc_4K_Metalness
  }
  n72 = LinearInterpolate(A=ScalarParameter() {parameterName=MetallicLow}, B=ScalarParameter() {defaultValue=1, parameterName=MetallicHigh}, Alpha=@n71.R)
  n73 = ScalarParameter() {defaultValue=0.8, parameterName=MetallicCaustics}
  n74 = NamedRerouteUsage()
  NamedRerouteDeclaration(None=LinearInterpolate(
    A=LinearInterpolate(A=@n72, B=@n73, Alpha=@n74)
    B=ScalarParameter() {defaultValue=0.9, parameterName=MetallicTransition}
    Alpha=NamedRerouteUsage()
   )) {name=Metallic, nodeColor=<rgba 0.5 1 0 1>}
  n75 = NamedRerouteUsage()
  n76 = AppendVector(A=ScalarParameter() {defaultValue=1, parameterName=Chrome_Scale_Ua}, B=ScalarParameter() {defaultValue=1, parameterName=Chrome_Scale_Va})
  NamedRerouteDeclaration(None=Panner(Coordinate=Multiply(A=@n75, B=@n76)) {bFractionalPart=true, speedX=0.0025, speedY=0.0075}) {name=Chrome_UVs_A, nodeColor=<rgba 0.9219 0 1 1>}
  n77 = NamedRerouteUsage()
  n78 = AppendVector(A=ScalarParameter() {defaultValue=2, parameterName=Chrome_Scale_Ub}, B=ScalarParameter() {defaultValue=1, parameterName=Chrome_Scale_Vb})
  NamedRerouteDeclaration(None=Panner(Coordinate=Multiply(A=@n77, B=@n78)) {bFractionalPart=true, speedX=0.015, speedY=0.0055}) {name=Chrome_UVs_B, nodeColor=<rgba 0 0.3906 1 1>}
  NamedRerouteDeclaration(None=Transform(None=@n42)) {name=ChromeNormals, nodeColor=<rgba 1 0 0.5391 1>}
  n79 = Power(Base=@n61, Exp=ScalarParameter() {defaultValue=3, parameterName=CausticsLerpExponent}) {constExponent=4}
  n80 = ScalarParameter() {defaultValue=3, parameterName=CausticsLerpScale}
  NamedRerouteDeclaration(None=Saturate(None=Multiply(A=@n79, B=@n80))) {name=CausticsLerp, nodeColor=<rgba 0.8984 0 1 1>}
  NamedRerouteUsage()
  NamedRerouteUsage()
  NamedRerouteUsage()
  n81 = Desaturation(
   None=TextureSampleParameter2D() {parameterName=SourceAlbedo, texture=&TBlackOut2_D}.RGB
   Fraction=Constant() {r=1}
  )
  n82 = DDX(Value=@n81)
  n83 = DDY(Value=@n81)
  Saturate(None=Multiply(A=Max(A=@n82, B=@n83), B=ScalarParameter() {defaultValue=20, parameterName=EdgeDetectionScale}) {constB=20})
  ScalarParameter() {defaultValue=0.5, parameterName=Specular}
 ]
}
