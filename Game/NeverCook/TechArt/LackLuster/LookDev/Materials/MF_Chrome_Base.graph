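# Material function that builds an animated chrome surface, optionally blended over incoming material attributes.
# Streak texture plus Fresnel indexes a curve-atlas gradient, combined with hue-shifted albedo for base color.
# World-aligned UVs by default drive panning swirls and layered detail, ripple and optional hexagon normals, with roughness varying on downward faces.
# The IsPhased switch enables panning reflection-vector emissive and sine-driven world-position jiggle, the latter scaled by PhaseActive.
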MaterialFunction /Game/NeverCook/TechArt/LackLuster/LookDev/Materials/MF_Chrome_Base
graph {nodes [
  n1 = ComponentMask(None=ReflectionVectorWS()) {b=true, r=true}
  n2 = Add(A=Divide(A=NamedRerouteUsage(), B=Constant2Vector() {g=1400, r=2600}) {constB=1400}, B=Power(Base=Abs(None=@n1)) {constExponent=3})
  n3 = TextureSample(UVs=Panner(Coordinate=@n2) {speedX=0.061, speedY=0.01}) {samplerSource=SSM_Wrap_WorldGroupSettings, texture=/Game/NeverCook/TechArt/LackLuster/LookDev/Textures/Phased_output}
  n4 = Constant2Vector() {g=-1, r=1}
  n5 = TextureSample(UVs=Panner(Coordinate=Multiply(A=@n2, B=@n4) {constB=-1}) {bFractionalPart=true, speedX=0.045, speedY=0.013}) {samplerSource=SSM_Wrap_WorldGroupSettings, texture=&Phased_output}
  Add(A=@n3.RGB, B=@n5.RGB)
  n6 = ComponentMask(None=NamedRerouteUsage()) {g=true, r=true}
  n7 = TextureSample(UVs=Add(A=@n6, B=NamedRerouteUsage())) {samplerType=SAMPLERTYPE_LinearGrayscale, texture=/Game/NeverCook/TechArt/LackLuster/LookDev/Textures/ChromeSwirl_Streaks}
  n8 = MaterialFunctionCall(
   Exponent=ScalarParameter() {defaultValue=1.5, parameterName=FresnelHueShiftExponent}
   Vertex / Pixel Normal=Transform(None=NamedRerouteUsage())
  ) {materialFunction=&MF_Fresnel_FX}
  n9 = CurveAtlasRowParameter(CurveTime=Frac(None=Add(A=@n7.R, B=@n8.Fresnel))) {
   atlas=/Game/NeverCook/TechArt/LackLuster/LookDev/Textures/CA_Chrome_RnD
   curve=/Game/NeverCook/TechArt/LackLuster/LookDev/Textures/C_Chrome_06
   defaultValue=6
   parameterName=Chrome
  }
  n10 = VectorParameter() {parameterName=AlbedoScale, defaultValue=<rgba 0.8385 0.8193 0.6886 0>}
  n11 = Frac(None=Multiply(A=@n8.Fresnel, B=ScalarParameter() {defaultValue=1, parameterName=HueShiftScale}))
  n12 = VectorParameter() {parameterName=AlbedoHueShift, defaultValue=<rgba 0.8212 0.9028 1 0>}
  n13 = Add(
   A=Multiply(A=@n9, B=NamedRerouteUsage())
   B=Multiply(
    A=FunctionInput(Preview=@n10.RGB) {bUsePreviewValueAsDefault=true, inputName=AlbedoOverride}
    B=MaterialFunctionCall(Hue Shift Percentage=@n11, Texture=@n12.RGB) {materialFunction=&HueShift}.Result
   )
  )
  n14 = Add(A=@n13, B=NamedRerouteUsage())
  Add(A=@n14, B=NamedRerouteUsage())
  AppendVector(A=ScalarParameter() {defaultValue=1, parameterName=PhaseU_Divisor}, B=ScalarParameter() {defaultValue=1, parameterName=PhaseV_Divisor})
  n15 = NamedRerouteUsage()
  ComponentMask(None=@n15) {g=true, r=true}
  n16 = TextureSample(UVs=NamedRerouteUsage()) {
   samplerSource=SSM_Wrap_WorldGroupSettings
   samplerType=SAMPLERTYPE_LinearColor
   texture=/Game/NeverCook/TechArt/LackLuster/LookDev/Textures/ChromeSwirl_output
  }
  n17 = TextureSample(UVs=NamedRerouteUsage()) {
   samplerSource=SSM_Wrap_WorldGroupSettings
   samplerType=SAMPLERTYPE_LinearColor
   texture=&ChromeSwirl_output
  }
  n18 = Divide(A=Add(A=@n16.R, B=@n17.R))
  CurveAtlasRowParameter(CurveTime=Frac(None=Add(A=@n18))) {
   atlas=&CA_Chrome_RnD
   curve=/Game/NeverCook/TechArt/LackLuster/LookDev/Textures/C_Chrome_02
   defaultValue=1
   parameterName=ChromeOld
  }
  Divide(
   A=ComponentMask(None=ReflectionVectorWS(CustomWorldNormal=@n15)) {g=true, r=true}
   B=ScalarParameter() {defaultValue=1, parameterName=ChromeReflectionVectorDIvisor}
  ) {constB=4}
  Divide(A=NamedRerouteUsage(), B=ScalarParameter() {defaultValue=500, parameterName=PhaseBaseUVDivisor})
  n19 = NamedRerouteDeclaration(None=FunctionInput(Preview=SetMaterialAttributes!(MaterialAttributes)) {bUsePreviewValueAsDefault=true, inputName=MA_In, inputType=FunctionInput_MaterialAttributes}) {name=MA_In, nodeColor=<rgba 0 0.1562 1 1>}
  n20 = NamedRerouteUsage()
  n21 = Time()
  n22 = Multiply(
   A=MaterialFunctionCall(R=Sine(None=@n21) {period=0.035}, G=Sine(None=@n21) {period=0.04}) {materialFunction=&AppendMany}.RGB
   B=Transform(None=VectorParameter() {parameterName=PhaseJiggleViewSpace, defaultValue=<rgba 0.4 0 0 0>}.RGB) {transformSourceType=TRANSFORMSOURCE_View}
  )
  n23 = ScalarParameter() {parameterName=PhaseActive}
  n24 = SetMaterialAttributes!(
   MaterialAttributes
   BaseColor=Desaturation(None=@n14, Fraction=ScalarParameter() {defaultValue=-0.5, parameterName=Fraction})
   Metallic=NamedRerouteUsage()
   Specular=ScalarParameter() {parameterName=Specular}
   Roughness=NamedRerouteUsage()
   Anisotropy=ScalarParameter() {defaultValue=0.03, parameterName=Anisotropy}
   OpacityMask=NamedRerouteUsage()
   Normal=@n20
   Tangent=@n20
   EmissiveColor=NamedRerouteUsage()
   WorldPositionOffset=StaticSwitchParameter(True=Multiply(A=@n22, B=@n23), False=Constant()) {parameterName=IsPhased}
  ) {attributeSetTypes [
    #1
    #2
    #3
    #4
    #5
    #6
    #7
    #8
    #9
    #10
   ]}
  n25 = FunctionInput() {bUsePreviewValueAsDefault=true, inputName=BlendMaterialAttributesAlpha, inputType=FunctionInput_Scalar}
  n26 = FunctionInput(Preview=StaticBool()) {bUsePreviewValueAsDefault=true, inputName=bBlendMaterialAttributes, inputType=FunctionInput_StaticBool}
  FunctionOutput(None=StaticSwitch(
    True=BlendMaterialAttributes(A=@n19, B=@n24, Alpha=@n25)
    False=@n24
    Value=NamedRerouteDeclaration(None=@n26) {name=bBlendMaterialAttributes, nodeColor=<rgba 1 0 0.4453 1>}
   )) {outputName=MA Out}
  Multiply()
  Multiply(A=@n6, B=ScalarParameter() {defaultValue=1, parameterName=ChromeDistortionScale})
  Multiply(A=PixelNormalWS(), B=VectorParameter() {parameterName=PhasedNormalReflectionStrength, defaultValue=<rgba 1 1 1 0>}.RGB)
  Multiply(B=SphereMask(
    A=WorldPosition().XYZ
    B=CameraPositionWS()
    Radius=ScalarParameter() {defaultValue=1000, parameterName=SparkleFadeDistance}
    Hardness=ScalarParameter() {defaultValue=0.1, parameterName=SparkleFadeFalloff}
   ))
  Multiply(A=@n9, B=ScalarParameter() {defaultValue=2, parameterName=ChromeColorScale})
  Multiply(A=@n13, B=VectorParameter() {parameterName=CausticAlbedoScale, defaultValue=<rgba 0.91 0.8706 0.849 0>}.RGB)
  NamedRerouteDeclaration(None=Divide(A=NamedRerouteUsage(), B=ScalarParameter() {defaultValue=250, parameterName=DetailNormal_UVs_Divisor})) {name=DetailNormalUVs, nodeColor=<rgba 1 0.0234 0 1>}
  n27 = TextureSampleParameter2D(UVs=Divide(A=NamedRerouteUsage(), B=Constant2Vector() {g=1, r=2})) {
   parameterName=NormalDetail
   samplerSource=SSM_Wrap_WorldGroupSettings
   samplerType=SAMPLERTYPE_Normal
   texture=/Game/NeverCook/TechArt/LackLuster/LookDev/Textures/Textures_DetailNormalSmoothed
  }
  n28 = Constant3Vector() {constant=<rgba 1 1 0 0>}
  n29 = TextureSample(UVs=NamedRerouteUsage()) {
   samplerSource=SSM_Wrap_WorldGroupSettings
   samplerType=SAMPLERTYPE_Normal
   texture=/Game/NeverCook/TechArt/LackLuster/LookDev/Textures/ChromeSwirl_Normal
  }
  n30 = TextureSample(UVs=NamedRerouteUsage()) {
   samplerSource=SSM_Wrap_WorldGroupSettings
   samplerType=SAMPLERTYPE_Normal
   texture=&ChromeSwirl_Normal
  }
  n31 = Multiply(
   A=VectorParameter() {parameterName=EdgeNormalScale, defaultValue=<rgba 0.18 0.18 1 0>}.RGB
   B=Normalize(VectorInput=Add(A=@n29.RGB, B=@n30.RGB))
  )
  n32 = ScalarParameter() {defaultValue=2.5, parameterName=EdgeNormal_FresnelExponent}
  n33 = VertexNormalWS()
  n34 = NamedRerouteDeclaration(None=LinearInterpolate(Alpha=MaterialFunctionCall(Exponent=@n32, Vertex / Pixel Normal=@n33) {materialFunction=&MF_Fresnel_FX}.Fresnel Inverted) {constA=0.05}) {name=EdgeNormalFresnelInverted, nodeColor=<rgba 0 1 0.5547 1>}
  n35 = Constant() {r=0.5}
  n36 = ScalarParameter() {defaultValue=0.35, parameterName=EdgeNormalFresnelRadius}
  n37 = Constant()
  n38 = Add(
   A=Add(A=NamedRerouteUsage(), B=Multiply(A=@n27.RGB, B=@n28))
   B=LinearInterpolate(B=@n31, Alpha=SphereMask(A=@n34, B=@n35, Radius=@n36, Hardness=@n37))
  )
  n39 = Multiply(
   A=TextureSample(UVs=NamedRerouteUsage()) {samplerType=SAMPLERTYPE_Normal, texture=/Game/NeverCook/TechArt/LackLuster/LookDev/Textures/Edge_NormalHexagons}.RGB
   B=VectorParameter() {parameterName=HexagonNormalScale, defaultValue=<rgba 1 1 0 0>}.RGB
  )
  n40 = NamedRerouteUsage()
  n41 = Add(A=@n38, B=Add(A=Multiply(A=@n39, B=@n40), B=NamedRerouteUsage()))
  NamedRerouteDeclaration(None=Normalize(VectorInput=StaticSwitchParameter(True=@n41, False=@n38) {parameterName=UseHexagons})) {name=Normals, nodeColor=<rgba 1 0 0.9141 1>}
  NamedRerouteDeclaration(None=Divide(A=NamedRerouteUsage(), B=ScalarParameter() {defaultValue=560, parameterName=Phase_UVs_Divisor})) {name=PhaseUVs, nodeColor=<rgba 0.9453 1 0 1>}
  n42 = NamedRerouteUsage()
  n43 = TextureSampleParameter2D!(UVs, Level, Apply View MipBias) {mipValueMode=TMVM_MipLevel, parameterName=Diffuse, texture=/Game/Weapons/WeaponSkins/Wraps/Textures/BlackOut1/TBlackOut2_D}
  n44 = TextureSampleParameter2D() {mipValueMode=TMVM_MipLevel, parameterName=Diffuse_Texture_2, texture=&TBlackOut2_D}
  n45 = TextureSampleParameter2D() {mipValueMode=TMVM_MipLevel, parameterName=Diffuse_Texture_3, texture=&TBlackOut2_D}
  n46 = TextureSampleParameter2D() {mipValueMode=TMVM_MipLevel, parameterName=Diffuse_Texture_4, texture=&TBlackOut2_D}
  n47 = StaticBool() {value=true}
  n48 = Desaturation(
   None=StaticSwitch(
    True=GetMaterialAttributes(None=@n42) {attributeGetTypes [#1]}.BaseColor
    False=MaterialFunctionCall!(transitionFactor, 1=@n43.RGB, 2=@n44.RGB, 3=@n45.RGB, 4=@n46.RGB, Use 2 Textures=@n47, Use 3 Textures=@n47, Use 4 Textures=@n47) {materialFunction=/Game/Athena/WinterArt/Materials/MF_LerpTexturesBasedOnInt_Float3_4Layers}.Result
    Value=NamedRerouteUsage()
   )
   Fraction=ScalarParameter() {defaultValue=0.92, parameterName=DiffuseDesaturation}
  )
  n49 = Add(
   A=Divide(A=VectorParameter() {parameterName=Diffuse_ColorDodgeBase, defaultValue=<rgba 0.4601 0.4251 0.4002 0>}.RGB, B=OneMinus(None=@n48))
   B=ScalarParameter() {defaultValue=-0.5, parameterName=DiffuseBias}
  )
  n50 = ScalarParameter() {defaultValue=1.6, parameterName=DiffuseScale}
  NamedRerouteDeclaration(None=Saturate(None=Multiply(A=@n49, B=@n50))) {name=Diffuse, nodeColor=<rgba 1 0 0.1172 1>}
  n51 = NamedRerouteUsage()
  n52 = Multiply(A=ScalarParameter() {defaultValue=1, parameterName=PhaseMaskSpeedScale}, B=Time())
  n53 = TextureSample(UVs=Panner(Coordinate=@n51, Time=@n52) {bFractionalPart=true, speedX=-0.0428, speedY=-0.0288}) {samplerSource=SSM_Wrap_WorldGroupSettings, samplerType=SAMPLERTYPE_LinearColor, texture=/Game/Global/Textures/Default/Noise/T_Noise01}
  n54 = TextureSample(UVs=Panner(Coordinate=@n51, Time=@n52) {speedX=0.0465, speedY=0.0338}) {samplerSource=SSM_Wrap_WorldGroupSettings, samplerType=SAMPLERTYPE_LinearColor, texture=&T_Noise01}
  n55 = SphereMask(
   A=Multiply(A=@n53.R, B=@n54.R)
   B=ScalarParameter() {parameterName=PhaseMaskCenter}
   Radius=ScalarParameter() {defaultValue=0.4, parameterName=PhaseMaskRadius}
   Hardness=ScalarParameter() {parameterName=PhaseMaskHardness}
  )
  NamedRerouteDeclaration(None=SphereMask(
    A=@n55
    B=ScalarParameter() {defaultValue=0.5, parameterName=PhaseMaskTransitionCenter}
    Radius=ScalarParameter() {defaultValue=0.5, parameterName=PhaseMaskTransitionRadius}
    Hardness=ScalarParameter() {defaultValue=0.1, parameterName=PhaseMaskTransitionHardness}
   )) {name=PhaseTransitionLerp, nodeColor=<rgba 0.0781 1 0 1>}
  NamedRerouteDeclaration(None=Divide(A=NamedRerouteUsage(), B=ScalarParameter() {defaultValue=800, parameterName=PhaseMaskUV_Divisor})) {name=PhaseMaskUVs, nodeColor=<rgba 1 0.9141 0 1>}
  n56 = Divide(A=NamedRerouteUsage(), B=ScalarParameter() {defaultValue=200, parameterName=SparklesDivisor})
  n57 = TextureSample(UVs=@n56) {
   samplerSource=SSM_Wrap_WorldGroupSettings
   samplerType=SAMPLERTYPE_LinearColor
   texture=/Game/NeverCook/TechArt/LackLuster/LookDev/Textures/Sparkle_Normal
  }
  n58 = ConstantBiasScale(None=@n57.RGB) {bias=-0.5, scale=2}
  n59 = DotProduct(A=Transform(None=@n58), B=CameraVectorWS())
  n60 = CurveAtlasRowParameter(CurveTime=Power(Base=Saturate(None=@n59), Exp=ScalarParameter() {defaultValue=50, parameterName=SparklesExponent}) {constExponent=20}) {
   atlas=&CA_Chrome_RnD
   curve=/Game/NeverCook/TechArt/LackLuster/LookDev/Textures/C_Chrome_07
   defaultValue=7
   parameterName=Sparkles
  }
  n61 = TextureSample(UVs=@n56) {
   samplerSource=SSM_Wrap_WorldGroupSettings
   samplerType=SAMPLERTYPE_LinearColor
   texture=/Game/NeverCook/TechArt/LackLuster/LookDev/Textures/Sparkle_Mask
  }
  NamedRerouteDeclaration(None=Multiply(A=@n60, B=@n61.R)) {name=Sparkles, nodeColor=<rgba 1 0.5391 0 1>}
  n62 = TextureSampleParameter2D!(UVs, Apply View MipBias) {
   parameterName=Normals
   samplerSource=SSM_Wrap_WorldGroupSettings
   samplerType=SAMPLERTYPE_Normal
   texture=&FlatNormalBlue
  }
  n63 = TextureSampleParameter2D() {
   parameterName=Normals_Texture_2
   samplerSource=SSM_Wrap_WorldGroupSettings
   samplerType=SAMPLERTYPE_Normal
   sortPriority=-5
   texture=&FlatNormalBlue
  }
  n64 = TextureSampleParameter2D() {
   parameterName=Normals_Texture_3
   samplerSource=SSM_Wrap_WorldGroupSettings
   samplerType=SAMPLERTYPE_Normal
   sortPriority=-5
   texture=&FlatNormalBlue
  }
  n65 = TextureSampleParameter2D() {
   parameterName=Normals_Texture_4
   samplerSource=SSM_Wrap_WorldGroupSettings
   samplerType=SAMPLERTYPE_Normal
   sortPriority=-5
   texture=/Engine/EngineMaterials/FlatNormal
  }
  n66 = StaticBool() {value=true}
  n67 = StaticSwitch(
   True=GetMaterialAttributes(None=NamedRerouteUsage()) {attributeGetTypes [#7]}.Normal
   False=StaticSwitchParameter(
    True=MaterialFunctionCall(1=@n62.RGB, 2=@n63.RGB, 3=@n64.RGB, 4=@n65.RGB, Use 2 Textures=@n66, Use 3 Textures=@n66, Use 4 Textures=@n66) {materialFunction=&MF_LerpTexturesBasedOnInt_Float3_4Layers}.Result
    False=@n62.RGB
   ) {defaultValue=true, parameterName=4Layer_Normals}
   Value=NamedRerouteUsage()
  )
  NamedRerouteDeclaration(None=@n67) {name=Normals4Layer, nodeColor=<rgba 0.5469 1 0 1>}
  n68 = MaterialFunctionCall!(UseLocalSpace, WorldPosition=WorldPosition().XYZ, NormalOverride) {materialFunction=/Game/Environments/Apollo/Sets/Skyfire/Materials/MaterialFunctions/MF_WorldAlignedUVs_HardEdge}
  n69 = TextureCoordinate()
  NamedRerouteDeclaration(None=FunctionInput(Preview=StaticSwitchParameter(True=@n68.Result, False=@n69) {defaultValue=true, parameterName=UseCheapTriplanarUVs}) {bUsePreviewValueAsDefault=true, inputName=WorldUVsOverride, inputType=FunctionInput_Vector2}) {name=UVs, nodeColor=<rgba 0.2891 0 1 1>}
  n70 = NamedRerouteUsage()
  n71 = AppendVector(
   A=ScalarParameter() {defaultValue=200, parameterName=Ripples_Scale_U1}
   B=ScalarParameter() {defaultValue=180, parameterName=Ripples_Scale_V1}
  )
  n72 = ScalarParameter() {defaultValue=0.0021, parameterName=Ripples_PanSpeed_U1}
  n73 = ScalarParameter() {defaultValue=0.0068, parameterName=Ripples_PanSpeed_V1}
  NamedRerouteDeclaration(None=Panner(Coordinate=Divide(A=@n70, B=@n71), Speed=AppendVector(A=@n72, B=@n73)) {bFractionalPart=true}) {name=RipplesUV1, nodeColor=<rgba 0 0.7188 1 1>}
  n74 = NamedRerouteUsage()
  n75 = AppendVector(
   A=ScalarParameter() {defaultValue=-210, parameterName=Ripples_Scale_U2}
   B=ScalarParameter() {defaultValue=-140, parameterName=Ripples_Scale_V2}
  )
  n76 = ScalarParameter() {defaultValue=0.0077, parameterName=Ripples_PanSpeed_U2}
  n77 = ScalarParameter() {defaultValue=-0.0149, parameterName=Ripples_PanSpeed_V2}
  NamedRerouteDeclaration(None=Panner(Coordinate=Divide(A=@n74, B=@n75), Speed=AppendVector(A=@n76, B=@n77)) {bFractionalPart=true}) {name=RipplesUV2, nodeColor=<rgba 0 0.7188 1 1>}
  n78 = TextureSample(UVs=NamedRerouteUsage()) {
   samplerSource=SSM_Wrap_WorldGroupSettings
   samplerType=SAMPLERTYPE_Normal
   texture=/Game/NeverCook/TechArt/LackLuster/LookDev/Textures/ChromeSwirl_RipplesNormal
  }
  n79 = TextureSample(UVs=NamedRerouteUsage()) {
   samplerSource=SSM_Wrap_WorldGroupSettings
   samplerType=SAMPLERTYPE_Normal
   texture=&ChromeSwirl_RipplesNormal
  }
  n80 = Add(A=@n78.RGB, B=Multiply(A=@n79.RGB, B=Constant3Vector() {constant=<rgba 1 1 0 0>}))
  NamedRerouteDeclaration(None=Multiply(A=Normalize(VectorInput=@n80), B=VectorParameter() {parameterName=RipplesNormalWallStrength, defaultValue=<rgba 0.35 0.35 0 0>}.RGB)) {name=RipplesNormals, nodeColor=<rgba 0 1 0.7656 1>}
  n81 = SkyLightEnvMapSample(Direction=ReflectionVectorWS(), Roughness=ScalarParameter() {defaultValue=0.2, parameterName=SkylightRoughness})
  n82 = VectorParameter() {parameterName=SkyLightFixColor, defaultValue=<rgba 0.0375 0.0721 0.151 0>}
  n83 = NamedRerouteUsage()
  n84 = Saturate(None=DotProduct(A=Transform(None=@n83), B=Constant3Vector() {constant=<rgba 0 0 -1 0>}))
  n85 = ScalarParameter() {defaultValue=0.15, parameterName=SkyLightFixCenter}
  n86 = ScalarParameter() {defaultValue=0.15, parameterName=SkyLightFixRadius}
  n87 = ScalarParameter() {parameterName=SkyLightFixHardness}
  NamedRerouteDeclaration(None=Multiply(A=Multiply(A=@n81, B=@n82.RGB), B=SphereMask(A=@n84, B=@n85, Radius=@n86, Hardness=@n87))) {name=SkyLightFix, nodeColor=<rgba 0.3594 1 0 1>}
  n88 = NamedRerouteUsage()
  n89 = TextureSampleParameter2D() {parameterName=OpacityMask, texture=/Game/Creative/BuildingActors/Props/Textures/CP_White_Texutre}
  NamedRerouteDeclaration(None=Multiply(
    A=GetMaterialAttributes(None=@n88) {attributeGetTypes [#6]}.OpacityMask
    B=ChannelMaskParameter(None=@n89.RGBA) {parameterName=OpacityMaskChannel}.R
   )) {name=OpacityMask, nodeColor=<rgba 0.3828 1 0 1>}
  n90 = DotProduct(A=VertexNormalWS(), B=Constant3Vector() {constant=<rgba 0 0 -1 0>})
  NamedRerouteDeclaration(None=SphereMask(
    A=Saturate(None=@n90)
    B=ScalarParameter() {defaultValue=0.1, parameterName=VertexNormalMaskCenter}
    Radius=ScalarParameter() {defaultValue=0.2, parameterName=VertexNormalMaskRadius}
    Hardness=ScalarParameter() {defaultValue=0.5, parameterName=VertexNormalMaskHardness}
   )) {name=VertexNormalMask.-Z, nodeColor=<rgba 0 0.3203 1 1>}
  NamedRerouteDeclaration(None=Divide(A=NamedRerouteUsage(), B=ScalarParameter() {defaultValue=250, parameterName=HexagonsUV_Divisor})) {name=HexagonUVs, nodeColor=<rgba 1 0 0.5625 1>}
  n91 = NamedRerouteUsage()
  n92 = TextureSample(UVs=@n91) {
   samplerSource=SSM_Wrap_WorldGroupSettings
   samplerType=SAMPLERTYPE_LinearColor
   texture=/Game/NeverCook/TechArt/LackLuster/LookDev/Textures/ChromeTech_Caustics
  }
  n93 = TextureSample(UVs=@n91) {
   samplerSource=SSM_Wrap_WorldGroupSettings
   samplerType=SAMPLERTYPE_LinearColor
   texture=&ChromeTech_Caustics
  }
  n94 = ScalarParameter() {defaultValue=3.2, parameterName=PhaseTimeDivisor}
  n95 = Add(
   A=Multiply(A=@n8.Fresnel, B=ScalarParameter() {defaultValue=1.5, parameterName=PhaseFresnelShiftScale}) {constB=3}
   B=Add(
    A=@n93.R
    B=MaterialFunctionCall!(Time, Divisor=@n94, Mod) {materialFunction=/Game/Effects/Fort_Effects/Materials/_Functions/MF_ModuloTime}.Clamped 0-1
   )
  )
  n96 = CurveAtlasRowParameter(CurveTime=Frac(None=Add(A=@n92.G, B=@n95))) {
   atlas=&CA_Chrome_RnD
   curve=/Game/NeverCook/TechArt/LackLuster/LookDev/Textures/C_Chrome_05
   defaultValue=4
   parameterName=PhaseBase
  }
  NamedRerouteDeclaration(None=@n96) {name=PhaseBase, nodeColor=<rgba 0 0.6016 1 1>}
  n97 = TextureSample(UVs=Panner(Coordinate=@n2) {bFractionalPart=true, speedX=0.7056, speedY=-0.0154}) {samplerSource=SSM_Wrap_WorldGroupSettings, texture=&Phased_output}
  n98 = Constant2Vector() {g=-1, r=-0.66}
  n99 = TextureSample(UVs=Panner(Coordinate=Multiply(A=@n2, B=@n98)) {bFractionalPart=true, speedX=-1.2908, speedY=-0.0195}) {samplerSource=SSM_Wrap_WorldGroupSettings, texture=&Phased_output}
  n100 = ScalarParameter() {defaultValue=0.2, parameterName=PhaseEmissiveScaleLow}
  n101 = ScalarParameter() {defaultValue=0.5, parameterName=PhaseEmissiveScaleHigh}
  n102 = MaterialFunctionCall(Exponent=ScalarParameter() {defaultValue=2, parameterName=PhaseFresnelExponent}) {materialFunction=&MF_Fresnel_FX}
  n103 = ScalarParameter() {parameterName=PhaseActive}
  n104 = Multiply(A=Multiply(A=Add(A=@n97.RGB, B=@n99.RGB), B=LinearInterpolate(A=@n100, B=@n101, Alpha=@n102.Fresnel Inverted)), B=@n103)
  n105 = Max(A=@n3.RGB, B=@n5.RGB)
  n106 = ScalarParameter() {defaultValue=0.5, parameterName=PhaseBaseDesaturation}
  n107 = Multiply(
   A=OneMinus(None=@n103)
   B=Multiply(A=Desaturation(None=@n105, Fraction=@n106), B=ScalarParameter() {defaultValue=0.1, parameterName=PhaseBaseEmissiveScale})
  )
  NamedRerouteDeclaration(None=StaticSwitchParameter(True=Add(A=@n104, B=@n107), False=Constant()) {parameterName=IsPhased}) {name=Emissive, nodeColor=<rgba 1 0 0.8203 1>}
  NamedRerouteDeclaration(None=LinearInterpolate(
    A=ScalarParameter() {defaultValue=0.28, parameterName=Roughness.-Z}
    B=ScalarParameter() {defaultValue=0.22, parameterName=Roughness}
    Alpha=NamedRerouteUsage()
   )) {name=Roughness, nodeColor=<rgba 1 0.3984 0 1>}
  NamedRerouteDeclaration(None=ScalarParameter() {defaultValue=0.965, parameterName=Metallic}) {name=Metallic, nodeColor=<rgba 0.5 1 0 1>}
  n108 = Divide(A=NamedRerouteUsage(), B=ScalarParameter() {defaultValue=1680, parameterName=Chrome_Scale_UV_a})
  n109 = Multiply(A=ScalarParameter() {defaultValue=1, parameterName=ChromePanSpeedScale}, B=Time())
  NamedRerouteDeclaration(None=Panner(Coordinate=NamedRerouteDeclaration(None=@n108) {name=Chrome_UVs_A_NoPanning, nodeColor=<rgba 0 1 0.7656 1>}, Time=@n109) {bFractionalPart=true, speedX=0.0008, speedY=0.0022}) {name=Chrome_UVs_A, nodeColor=<rgba 0.9219 0 1 1>}
  n110 = NamedRerouteUsage()
  n111 = ScalarParameter() {defaultValue=2352, parameterName=Chrome_Scale_UV_b}
  NamedRerouteDeclaration(None=Panner(Coordinate=Divide(A=@n110, B=@n111), Time=@n109) {bFractionalPart=true, speedX=0.0054, speedY=0.0022}) {name=Chrome_UVs_B, nodeColor=<rgba 0 0.3906 1 1>}
  NamedRerouteDeclaration(None=Transform(None=@n31)) {name=ChromeNormalsWS, nodeColor=<rgba 1 0 0.5391 1>}
  n112 = VectorParameter() {parameterName=SourceNormalsStrength, defaultValue=<rgba 0.25 0.25 1 0>}
  NamedRerouteDeclaration(None=Multiply(A=@n67, B=FunctionInput(Preview=@n112.RGB) {bUsePreviewValueAsDefault=true, inputName=SourceNormalsStrengthOverride})) {name=Normals4layerScaled, nodeColor=<rgba 0.5469 1 0 1>}
  NamedRerouteDeclaration(None=@n55) {name=PhaseLerp, nodeColor=<rgba 0.8984 0 1 1>}
  NamedRerouteUsage()
  NamedRerouteUsage()
  RotateAboutAxis(
   NormalizedRotationAxis=Constant3Vector() {constant=<rgba 0 0 1 0>}
   RotationAngle=MaterialFunctionCall(Divisor=Constant() {r=10}) {materialFunction=&MF_ModuloTime}.Unclamped
   PivotPoint=Constant() {r=1}
   Position=@n58
  )
  n113 = VectorParameter() {parameterName=CausticsNormalScale, defaultValue=<rgba 0.1 0.1 1 0>}
  StaticSwitchParameter(True=LinearInterpolate(A=@n38, B=Multiply(A=@n38, B=@n113.RGB), Alpha=NamedRerouteUsage()), False=@n38) {parameterName=IsPhased}
 ]}
