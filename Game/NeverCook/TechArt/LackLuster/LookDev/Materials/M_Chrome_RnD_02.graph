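# M_Chrome_RnD_02 is a look-development research material for a chrome surface, using Default Lit shading through material attributes.
# Its only graph node is a call to the MF_Chrome_Base material function, which holds the chrome logic outside this file.
# That function exposes inputs for attribute blending, world UV override, albedo override and source normal strength.
# Output translucent velocity is enabled, and pixel depth offset uses legacy mode.
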Material /Game/NeverCook/TechArt/LackLuster/LookDev/Materials/M_Chrome_RnD_02
graph {
 properties {bOutputTranslucentVelocity=true, bUseMaterialAttributes=true, pixelDepthOffsetMode=PDOM_Legacy, usedShadingModels=Default Lit}
 nodes [MaterialFunctionCall!(BlendMaterialAttributesAlpha, MA_In, WorldUVsOverride, AlbedoOverride, SourceNormalsStrengthOverride, bBlendMaterialAttributes) {materialFunction=&/MF_Chrome_Base}]
}
