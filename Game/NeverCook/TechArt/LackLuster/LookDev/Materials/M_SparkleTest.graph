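# M_SparkleTest is an unlit look-dev test material, named for a sparkle effect, with no material outputs connected.
# It divides texture coordinates by scaled camera-to-pixel distance, then feeds the screen-space derivatives of that result into a linear-color texture sample as explicit mip gradients.
# The scaled coordinates drive only those gradients, not the sample's UVs.
# A second, normal-map sample is unconnected, and neither sample has a texture assigned.
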Material /Game/NeverCook/TechArt/LackLuster/LookDev/Materials/M_SparkleTest
graph {
 properties {pixelDepthOffsetMode=PDOM_Legacy, shadingModel=MSM_Unlit, usedShadingModels=Unlit}
 nodes [
  TextureCoordinate()
  n1 = WorldPosition()
  n2 = CameraPositionWS()
  n3 = Divide(A=TextureCoordinate(), B=Multiply(A=Distance(A=@n1.XYZ, B=@n2)) {constB=0.05})
  TextureSample!(UVs, Tex, "DDX(UVs)"=DDX(Value=@n3), "DDY(UVs)"=DDY(Value=@n3), Apply View MipBias) {mipValueMode=TMVM_Derivative, samplerType=SAMPLERTYPE_LinearColor}
  TextureSample!(UVs, Tex, Apply View MipBias) {samplerType=SAMPLERTYPE_Normal}
 ]
}
