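# Unlit look-dev scratch material that tests an animated mask over random noise, with no material output connections shown.
# Looping Time, offset by a Radius parameter, sweeps the center of a SphereMask applied to MF_RandomUVSample noise on V-tiled UVs, selecting a moving band of noise values.
# DebugScalarValues receives the mask as a number to display on repeated UVs.
# An MF_Timestamp_InOut_01 call sits unconnected.
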Material /Game/NeverCook/TechArt/LackLuster/LookDev/Materials/NewMaterial
graph {
 properties {pixelDepthOffsetMode=PDOM_Legacy, shadingModel=MSM_Unlit, usedShadingModels=Unlit}
 nodes [
  n1 = TextureCoordinate() {vTiling=10}
  n2 = Time()
  n3 = Multiply(A=Frac(None=Multiply(A=@n2) {constB=0.3})) {constB=2}
  n4 = ScalarParameter() {defaultValue=0.5, parameterName=Radius}
  MaterialFunctionCall(
   Number=SphereMask(
    A=MaterialFunctionCall!(PrimaryTextureCoordinates=@n1, 2ndOrderTextureCoordinates, RandomSeed, NoiseDensityBias, Bypass Noise Influence Crosstalk) {materialFunction=/Game/Effects/Fort_Effects/Materials/_Functions/MF_RandomUVSample}.NoiseWithRange
    B=Subtract(A=@n3, B=@n4)
    Radius=@n4
    Hardness=Constant()
   )
   UVs=Frac(None=TextureCoordinate() {vTiling=10})
  ) {materialFunction=&DebugScalarValues}
  MaterialFunctionCall() {materialFunction=&MF_Timestamp_InOut_01}
 ]
}
